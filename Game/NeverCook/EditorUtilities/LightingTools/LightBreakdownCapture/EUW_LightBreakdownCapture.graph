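# Editor utility widget that captures a lighting breakdown of the current editor view through a TakeShot button.
# It finds or spawns a tagged transient CameraActor, aligns it to the viewport, and saves high-res screenshots under Saved/LightBreakdown/Capture.
# After a baseline shot, a delay-chained loop zeroes each listed show-flag console variable, captures, then restores it.
# The EventGraph also holds an unconnected sphere trace and PointLight cast.
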EditorUtilityWidgetBlueprint /Game/NeverCook/EditorUtilities/LightingTools/LightBreakdownCapture/EUW_LightBreakdownCapture
graphs [
 CleanupPostScreenshot {nodes [
   CleanupPostScreenshot() {is=title K2Node_FunctionEntry}
   > "Set Console Variable By Name (Int)"!(execute, InCommandName:String=Get VarToReset() -> (VarToReset:String), InValue:int=Get ValueToReset() -> (ValueToReset:int=0)) -> (then, ReturnValue:bool) {is=title K2Node_CallFunction}
   > Set VarToReset(VarToReset:String) -> (Output_Get:String)
   > Set ValueToReset(ValueToReset:int=-5) -> (Output_Get:int=0)
  ]}
 EventGraph {nodes [
   "On Clicked (TA_Generic01_Button)"!() -> (OutputDelegate, then, widget:User Widget Object Reference, event:Pointer Event Structure) {is=title K2Node_ComponentBoundEvent}
   > Set ShowFlagCurrentIndex(ShowFlagCurrentIndex:int=0) -> (Output_Get:int=0)
   > TakeScreenshot!(execute, self, name:String=Capture) -> (then)
   > Delay!(execute, WorldContextObject:Object Reference, Duration=1) -> (then)
   > TakeNextShowFlagShot()
   n1 = TakeDelayedScreenshot!() -> (OutputDelegate, then, name:String) {is=title K2Node_CustomEvent}
   > Delay(Duration=1)
   > TakeScreenshot(name=@n1.name)
   > Delay(Duration=1)
   > CleanupPostScreenshot() {is=title K2Node_CallFunction}
   > Delay(Duration=1)
   > TakeNextShowFlagShot()
   Multi Sphere Trace By Channel!(
    execute
    WorldContextObject:Object Reference
    Start:Vector="0, 0, 0"
    End:Vector="0, 0, 0"
    Radius:float=0
    TraceChannel:ETraceTypeQuery Enum
    bTraceComplex:bool=false
    ActorsToIgnore:Array of Actor Object References
    DrawDebugType:EDrawDebugTrace Enum
    bIgnoreSelf:bool=true
    TraceColor:Linear Color Structure=(R=1,G=0,B=0,A=1)
    TraceHitColor:Linear Color Structure=(R=0,G=1,B=0,A=1)
    DrawTime:float=5
   ) -> (then, OutHits:Array of Hit Result Structures, ReturnValue:bool)
   Cast To PointLight!(execute, Object:Wildcard) -> (then, CastFailed, AsPoint Light:Point Light Object Reference) {is=title K2Node_DynamicCast}
   Event PreConstruct()
   Event Construct()
   Event Tick()
  ]}
 TakeNextShowFlagShot {nodes [
   TakeNextShowFlagShot() {is=title K2Node_FunctionEntry}
   > Branch(Condition=Is Valid Index!(
     TargetArray=Get ShowFlagsToTest() -> (ShowFlagsToTest:Array of Strings)
     IndexToTest:int=Get ShowFlagCurrentIndex() -> (ShowFlagCurrentIndex:int=0)
    ) -> (ReturnValue:bool))
   > TakeShotShowFlagShot!(
    execute
    self
    VariableName:String="Get (a copy)"(Array=Get ShowFlagsToTest(), Dimension 1=Get ShowFlagCurrentIndex() -> (ShowFlagCurrentIndex=0))
   ) -> (then)
   > Increment Int(Value=Get ShowFlagCurrentIndex() -> (ShowFlagCurrentIndex=0))
  ]}
 TakeScreenshot {nodes [
   n1 = TakeScreenshot!() -> (then, name:String=Capture) {is=title K2Node_FunctionEntry}
   > n2 = Sequence!(execute) -> (then_0, then_1, then_2:Exec)
   > n3 = GetAllActorsOfClassWithTag!(execute, WorldContextObject:Object Reference, ActorClass:Actor Class Reference, Tag:Name=Get CameraTag() -> (CameraTag:Name=None)) -> (then, OutActors:Array of Actor Object References)
   > n4 = Branch(Condition=Is Valid Index(TargetArray:Array of Actor Object References=@n3))
   > n5 = Cast To CameraActor!(execute, Object="Get (a copy)"(Array:Array of Actor Object References=@n3) -> (Output:Actor Object Reference)) -> (then, CastFailed, AsCamera Actor:Camera Actor Object Reference) {is=title K2Node_DynamicCast}
   > Set l_camera(l_camera:Camera Actor Object Reference=@n5) -> (Output_Get:Camera Actor Object Reference)
   n6 = SpawnActorFromClass!(
    execute=@n4.else
    ActorClass:Actor Class Reference=/Script/Engine.CameraActor
    Location:Vector="0, 0, 0"
    Rotation:Rotator
    bTransient:bool=true
   ) -> (then, ReturnValue:Camera Actor Object Reference)
   > SetFieldOfView!(
    execute
    self:Camera Component Object Reference=Get CameraComponent!(self:Camera Actor Object Reference=@n6) -> (CameraComponent:Camera Component Object Reference)
    InFieldOfView:float=90
   ) -> (then)
   > Set l_camera(l_camera=@n6)
   > Set Tags!(
    execute
    Tags:Array of Names=Make Array("[0]":Name=Get CameraTag() -> (CameraTag=None)) -> (Array:Array of Names)
    self:Actor Object Reference=@n6
   ) -> (then, Output_Get:Array of Names)
   n7 = Get Editor Camera Info!(execute=@n2.then_1) -> (then, CameraLocation:Vector, ForwardDirection:Vector, FOV:float, ReturnValue:bool) {is=title K2Node_CallFunction}
   > Set Actor Location And Rotation!(
    execute
    self:Actor Object Reference=Get l_camera() -> (l_camera:Camera Actor Object Reference)
    NewLocation:Vector=@n7
    NewRotation:Rotator=MakeRotFromX!(X:"Vector (by ref)"=@n7.ForwardDirection) -> (ReturnValue:Rotator)
    bSweep:bool=false
    bTeleport:bool=true
   ) -> (then, SweepHitResult:Hit Result Structure, ReturnValue:bool) {is=title K2Node_CallFunction}
   TakeHighResScreenshot!(
    execute=@n2.then_2
    ResX:int=1024
    ResY:int=1024
    Filename:String=Append!(
     A=ProjectSavedDir!() -> (ReturnValue:String)
     B=LightBreakdown/Capture/
     C:String=getScreenshotNameFromCvar!(string:String=@n1) -> (friendly:String)
    ) -> (ReturnValue)
    Camera:Camera Actor Object Reference=Get l_camera()
    bMaskEnabled:bool=false
    bCaptureHDR:bool=false
    ComparisonTolerance:EComparisonTolerance Enum=Low
    ComparisonNotes:String
    Delay:float=0
    bForceGameView:bool=true
   ) -> (then, ReturnValue:Automation Editor Task Object Reference)
  ]}
 TakeShotShowFlagShot {nodes [
   n1 = TakeShotShowFlagShot!() -> (then, VariableName:String) {is=title K2Node_FunctionEntry}
   > n2 = Set VarToReset(VarToReset=@n1)
   > Set l_name(l_name:String=@n2) -> (Output_Get:String)
   > n3 = Sequence!(execute) -> (then_0, then_1, then_2, then_3:Exec)
   > n4 = GetConsoleVariableIntValue!(execute, VariableName:String=Get l_name() -> (l_name:String)) -> (then, ReturnValue:int=0)
   > n5 = Set l_value(l_value:int=@n4) -> (Output_Get:int=0)
   > Set ValueToReset(ValueToReset=@n5) -> (Output_Get=0)
   "Set Console Variable By Name (Int)"(execute=@n3.then_1, InCommandName=Get l_name(), InValue=0)
   TakeDelayedScreenshot!(execute=@n3.then_2, self, name:String=Get l_name()) -> (then) {is=title K2Node_CallFunction}
  ]}
 getScreenshotNameFromCvar {nodes [
   n1 = getScreenshotNameFromCvar!() -> (then, string:String) {is=title K2Node_FunctionEntry}
   n2 = TrimTrailing!(SourceString:String=Trim!(SourceString:String=@n1) -> (ReturnValue:String)) -> (ReturnValue:String)
   > Return Node!(execute, friendly:String=Replace(SourceString=Replace(SourceString=@n2, From=., To=_), From=" "))
  ]}
]
variables [
 CameraTag {type=string, default=LightBreakdownCaptureCamera}
 ShowFlagCurrentIndex {type=integer}
 ShowFlagsToTest {type="array<string>", default [ShowFlag.EyeAdaptation, ShowFlag.PostProcessing, ShowFlag.LumenGlobalIllumination, ShowFlag.LocalExposure]}
 ValueToReset {type=integer}
 VarToReset {type=string, default=""}
]
widgets {
 info {rootWidgetClass=&StackBox, widgetCount=2}
 widgets [
  StackBox_0 {widgetClassPath=&StackBox, widget=StackBox_0, properties {orientation=Orient_Vertical}}
  TA_Generic01_Button {
   widgetClassPath=&TA_Generic01_Button
   parent=StackBox_0
   widget=TA_Generic01_Button
   properties {label=TakeShot}
   slot {class=StackBoxSlot, properties {size {sizeRule=Fill, value=1}}}
  }
 ]
}
